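# UI-domain material in the Clyde Movies folder, evidently meant to show the ClydeMediaPlayer video in widgets.
# Its only node is a TextureSample referencing the ClydeMediaPlayer_Video texture.
# The graph shows no connection from that sample to a material output and no parameters.
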Material /Clyde/Movies/ClydeMediaPlayer_Video_Mat
graph {
 properties {materialDomain=MD_UI, pixelDepthOffsetMode=PDOM_Legacy, usedShadingModels=Default Lit}
 nodes [TextureSample() {texture=&/ClydeMediaPlayer_Video}]
}
